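# Physics asset for the Trey Cozy medium male player body, set up for the physics of its two drawstrings.
# Each string is a three-bone capsule chain with tight swing and twist limits, hung from a small master body constrained to the pelvis.
# Larger capsules on pelvis, thighs and spine_05 form the body volumes, with free angular constraints to the pelvis.
# The right master's pelvis joint is also angularly free, unlike the tightly limited left.
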PhysicsAsset /BRCosmetics/Characters/Player/Male/Medium/Bodies/M_MED_Trey_Cozy/Meshes/M_MED_Trey_Cozy_Physics
bodies {
 dyn_drawString01_l [{shapeType=Capsule, length=3.75, radius=0.75, shapeName=dyn_drawString01_l_capsule, center=<2.0005 0 0>, rotation=<rot 90 0 0>}]
 dyn_drawString01_l_MSTR [{
   shapeType=Capsule
   length=0.505
   radius=0.505
   shapeName=dyn_drawString01_l_MSTR_capsule
   center=<-0.1066 0 -0.1569>
   rotation=<rot 0 -89.9999 0>
  }]
 dyn_drawString01_r [{shapeType=Capsule, length=2.25, radius=1, shapeName=dyn_drawString01_r_capsule, center=<1.3705 0 0>, rotation=<rot 90 0 0>}]
 dyn_drawString01_r_MASTR [{shapeType=Capsule, length=0.505, radius=0.505, shapeName=dyn_drawString01_r_MASTR_capsule, rotation=<rot 0 90 0>}]
 dyn_drawString02_l [{shapeType=Capsule, length=3, radius=0.9, shapeName=dyn_drawString02_l_capsule, center=<2.1757 -0.0001 0>, rotation=<rot 90 0 0>}]
 dyn_drawString02_r [{shapeType=Capsule, length=2.25, radius=1, shapeName=dyn_drawString02_r_capsule, center=<1.8287 -0.0001 0>, rotation=<rot 90 0 0>}]
 dyn_drawString03_l [{shapeType=Capsule, length=2.25, radius=1, shapeName=dyn_drawString03_l_capsule, center=<1.6108 -0.0001 0>, rotation=<rot 90 0 0>}]
 dyn_drawString03_r [{shapeType=Capsule, length=2.25, radius=0.9, shapeName=dyn_drawString03_r_capsule, center=<1.8442 0 0>, rotation=<rot 90 0 0>}]
 pelvis [
  {shapeType=Capsule, length=17.3675, radius=9.2488, shapeName=pelvis_capsule, center=<-6.8691 1.6765 0>, rotation=<rot 0 89.9999 0>}
  {shapeType=Capsule, length=14.25, radius=7.5, shapeName=pelvis_capsule, center=<-12.4031 5.3394 0>}
 ]
 spine_05 [{shapeType=Capsule, length=17.5244, radius=12.1776, center=<-5.3132 0.0003 0>}]
 thigh_l [{shapeType=Capsule, length=21.75, radius=7.75, shapeName=thigh_l_capsule, center=<-15 1.9173 0.4059>, rotation=<rot 0 -89.9999 -89.9999>}]
 thigh_r [{shapeType=Capsule, length=22, radius=8, shapeName=thigh_r_capsule, center=<15 -1.1158 0.0216>, rotation=<rot 0 -89.9999 90>}]
}
constraints [
 {bone1Name=dyn_drawString02_r, bone2Name=dyn_drawString01_r, swing1LimitDegrees=10, swing2LimitDegrees=10, twistLimitDegrees=2}
 {bone1Name=dyn_drawString03_r, bone2Name=dyn_drawString02_r, swing1LimitDegrees=10, swing2LimitDegrees=10, twistLimitDegrees=2}
 {bone1Name=dyn_drawString02_l, bone2Name=dyn_drawString01_l, swing1LimitDegrees=10, swing2LimitDegrees=10, twistLimitDegrees=2}
 {bone1Name=dyn_drawString03_l, bone2Name=dyn_drawString02_l, swing1LimitDegrees=10, swing2LimitDegrees=10, twistLimitDegrees=2}
 {
  bone1Name=spine_05
  bone2Name=pelvis
  swing1LimitDegrees=45
  swing1Motion=Free
  swing2LimitDegrees=45
  swing2Motion=Free
  twistLimitDegrees=45
  twistMotion=Free
 }
 {bone1Name=dyn_drawString01_l, bone2Name=dyn_drawString01_l_MSTR, swing1LimitDegrees=10, swing2LimitDegrees=10, twistLimitDegrees=2}
 {
  bone1Name=thigh_l
  bone2Name=pelvis
  swing1LimitDegrees=45
  swing1Motion=Free
  swing2LimitDegrees=45
  swing2Motion=Free
  twistLimitDegrees=45
  twistMotion=Free
 }
 {
  bone1Name=thigh_r
  bone2Name=pelvis
  swing1LimitDegrees=45
  swing1Motion=Free
  swing2LimitDegrees=45
  swing2Motion=Free
  twistLimitDegrees=45
  twistMotion=Free
 }
 {bone1Name=dyn_drawString01_l_MSTR, bone2Name=pelvis, swing1LimitDegrees=10, swing2LimitDegrees=10, twistLimitDegrees=2}
 {bone1Name=dyn_drawString01_r, bone2Name=dyn_drawString01_r_MASTR, swing1LimitDegrees=10, swing2LimitDegrees=10, twistLimitDegrees=2}
 {
  bone1Name=dyn_drawString01_r_MASTR
  bone2Name=pelvis
  swing1LimitDegrees=45
  swing1Motion=Free
  swing2LimitDegrees=45
  swing2Motion=Free
  twistLimitDegrees=45
  twistMotion=Free
 }
]
