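# Post-process style cosmetic AnimBP for the M_MED_Trey_Cozy player body, adding secondary motion to shoe flaps, hoodie and drawstrings.
# It copies the attached parent mesh's pose, then runs a RigidBody simulation in pelvis space.
# Chained spring-driven AnimDynamics nodes follow, with right shoe flaps referencing foot_l.
# Additive bone transforms finish the pose, some drawstring offsets blending on bIsAccelerating2D; the EventGraph has no logic.
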AnimBlueprint /BRCosmetics/Characters/Player/Male/Medium/Bodies/M_MED_Trey_Cozy/Meshes/M_MED_Trey_Cozy_AnimBP
target_skeleton=&/M_MED_Trey_Cozy_Skeleton
graphs [
 AnimGraph {nodes [
   n1 = RigidBody(ComponentPose=Local To Component(LocalPose=Copy Pose From Mesh() {bUseAttachedParent=true})) {
    bOverrideWorldGravity=true
    overlapChannel=ECC_PhysicsBody
    overridePhysicsAsset=&/M_MED_Trey_Cozy_Physics
    simulationSpace=BaseBoneSpace
    alphaBoolBlend {blendOption=Linear}
    baseBoneRef {boneName=pelvis}
    componentLinearAccScale=<0.3 0.3 0.3>
    componentLinearVelScale=<0.3 0.3 0.3>
    overrideWorldGravity {z=-980}
   }
   n2 = Anim Dynamics - Bone: dyn_back_shoeFlap_top_l(ComponentPose=@n1, Alpha=1, ExternalForce="0,0,0") {
    angularDampingOverride=0.9999
    angularSpringConstant=12000
    bAngularSpring=true
    bGravityOverrideInSimSpace=true
    bOverrideAngularDamping=true
    bUseGravityOverride=true
    gravityScale=0.35
    simulationSpace=RootRelative
    alphaBoolBlend {blendOption=Linear}
    boundBone {boneName=dyn_back_shoeFlap_top_l}
    componentLinearAccScale=<1 1 1>
    componentLinearVelScale=<1 1 1>
    gravityOverride {z=-980}
    physicsBodyDefinitions [{
      sphereCollisionRadius=0
      boundBone {boneName=dyn_back_shoeFlap_top_l}
      boxExtents=<20 10 10>
      constraintSetup {angularLimitsMax=<0 15 20>, angularLimitsMin=<0 -15 -10>, angularTarget=<1 0 0>}
      localJointOffset=<4 0 0>
     }]
   }
   n3 = Anim Dynamics - Bone: dyn_back_shoeFlap_bottom_l(ComponentPose=@n2, Alpha=0.5, ExternalForce="0,0,0") {
    angularDampingOverride=0.999
    angularSpringConstant=12000
    bAngularSpring=true
    bGravityOverrideInSimSpace=true
    bOverrideAngularDamping=true
    bUseGravityOverride=true
    gravityScale=0.35
    simulationSpace=RootRelative
    alphaBoolBlend {blendOption=Linear}
    boundBone {boneName=dyn_back_shoeFlap_bottom_l}
    componentLinearAccScale=<1 1 1>
    componentLinearVelScale=<1 1 1>
    gravityOverride {z=-980}
    physicsBodyDefinitions [{
      sphereCollisionRadius=0
      boundBone {boneName=dyn_back_shoeFlap_bottom_l}
      boxExtents=<20 10 10>
      constraintSetup {angularLimitsMax=<15 0 0>, angularLimitsMin=<-15 0 0>, angularTarget=<1 0 0>}
      localJointOffset=<4 0 0>
     }]
   }
   n4 = Anim Dynamics - Bone: dyn_front_shoeFlap_l(ComponentPose=@n3, Alpha=1, ExternalForce="0,0,0") {
    angularDampingOverride=0.9
    bAngularSpring=true
    bGravityOverrideInSimSpace=true
    bOverrideAngularDamping=true
    bUseGravityOverride=true
    simulationSpace=BoneRelative
    alphaBoolBlend {blendOption=Linear}
    boundBone {boneName=dyn_front_shoeFlap_l}
    componentAppliedLinearAccClamp=<30000 30000 300000>
    componentLinearAccScale=<5 5 5>
    componentLinearVelScale=<5 5 5>
    gravityOverride {z=-1000}
    physicsBodyDefinitions [{
      sphereCollisionRadius=0
      boundBone {boneName=dyn_front_shoeFlap_l}
      boxExtents=<8 8 10>
      constraintSetup {twistAxis=AxisY, angularLimitsMax=<-40 0 0>, angularLimitsMin=<-45 0 0>, angularTarget=<0 0 0>}
      localJointOffset=<0 0 -0.2>
     }]
    relativeSpaceBone {boneName=foot_l}
   }
   n5 = Anim Dynamics - Bone: dyn_back_shoeFlap_top_r(ComponentPose=@n4, Alpha=1, ExternalForce="0,0,0") {
    angularDampingOverride=0.9999
    bAngularSpring=true
    bGravityOverrideInSimSpace=true
    bOverrideAngularDamping=true
    bUseGravityOverride=true
    simulationSpace=BoneRelative
    alphaBoolBlend {blendOption=Linear}
    boundBone {boneName=dyn_back_shoeFlap_top_r}
    componentAppliedLinearAccClamp=<30000 30000 300000>
    componentLinearAccScale=<0.05 0.05 0.05>
    componentLinearVelScale=<0.05 0.05 0.05>
    gravityOverride {z=-1000}
    physicsBodyDefinitions [{
      sphereCollisionRadius=0
      boundBone {boneName=dyn_back_shoeFlap_top_r}
      boxExtents=<2 2 2>
      constraintSetup {angularLimitsMax=<0 10 0>, angularLimitsMin=<0 -10 0>, angularTarget=<0 0 0>}
      localJointOffset=<0 0 -0.05>
     }]
    relativeSpaceBone {boneName=foot_l}
   }
   n6 = Anim Dynamics - Bone: dyn_back_shoeFlap_bottom_r(ComponentPose=@n5, Alpha=1, ExternalForce="0,0,0") {
    angularDampingOverride=0.999
    bAngularSpring=true
    bGravityOverrideInSimSpace=true
    bOverrideAngularDamping=true
    bUseGravityOverride=true
    simulationSpace=BoneRelative
    alphaBoolBlend {blendOption=Linear}
    boundBone {boneName=dyn_back_shoeFlap_bottom_r}
    componentAppliedLinearAccClamp=<30000 30000 300000>
    componentLinearAccScale=<0.4 0.4 0.4>
    componentLinearVelScale=<0.4 0.4 0.4>
    gravityOverride {z=-1000}
    physicsBodyDefinitions [{
      sphereCollisionRadius=0
      boundBone {boneName=dyn_back_shoeFlap_bottom_r}
      boxExtents=<5 5 7>
      constraintSetup {angularLimitsMax=<0 20 0>, angularLimitsMin=<0 -20 0>, angularTarget=<0 0 0>}
      localJointOffset=<0 0 -0.2>
     }]
    relativeSpaceBone {boneName=foot_l}
   }
   n7 = Anim Dynamics - Bone: dyn_front_shoeFlap_r(ComponentPose=@n6, Alpha=1, ExternalForce="0,0,0") {
    angularDampingOverride=0.9
    bAngularSpring=true
    bGravityOverrideInSimSpace=true
    bOverrideAngularDamping=true
    bUseGravityOverride=true
    simulationSpace=BoneRelative
    alphaBoolBlend {blendOption=Linear}
    boundBone {boneName=dyn_front_shoeFlap_r}
    componentAppliedLinearAccClamp=<30000 30000 300000>
    componentLinearAccScale=<5 5 5>
    componentLinearVelScale=<5 5 5>
    gravityOverride {z=-1000}
    physicsBodyDefinitions [{
      sphereCollisionRadius=0
      boundBone {boneName=dyn_front_shoeFlap_r}
      boxExtents=<8 8 10>
      constraintSetup {twistAxis=AxisY, angularLimitsMax=<-40 0 0>, angularLimitsMin=<-45 0 0>, angularTarget=<0 0 0>}
      localJointOffset=<0 0 -0.2>
     }]
    relativeSpaceBone {boneName=foot_l}
   }
   n8 = Anim Dynamics - Bone: dyn_hoodie!(ComponentPose=@n7, Alpha=1, GravityOverride:Vector="0,0,0", ExternalForce="0,0,0") -> (Pose) {
    angularDampingOverride=0.9999
    bAngularSpring=true
    bGravityOverrideInSimSpace=true
    bOverrideAngularDamping=true
    simulationSpace=RootRelative
    alphaBoolBlend {blendOption=Linear}
    boundBone {boneName=dyn_hoodie}
    componentAppliedLinearAccClamp=<3500 3500 3500>
    componentLinearAccScale {z=1.5}
    componentLinearVelScale {z=1.5}
    physicsBodyDefinitions [{
      sphereCollisionRadius=0
      boundBone {boneName=dyn_hoodie}
      boxExtents=<30 30 10>
      constraintSetup {angularLimitsMax=<0 1.5 0>, angularLimitsMin=<0 -20 0>, angularTarget=<0 0 0>}
      localJointOffset=<1 0 0>
     }]
    relativeSpaceBone {boneName=spine_05}
   }
   n9 = "Transform (Modify) Bone - Bone: dyn_helperBack"(ComponentPose=@n8, Alpha=1, Translation="2,-2,0", Rotation="0,0,0", Scale="1,1,1") {
    translationMode=BMM_Additive
    translationSpace=BCS_BoneSpace
    alphaBoolBlend {blendOption=Linear}
    boneToModify {boneName=dyn_helperBack}
    translation=<2 -2>
   }
   n10 = "Transform (Modify) Bone - Bone: dyn_drawString03_l"!(ComponentPose=@n9, bAlphaBoolEnabled:bool=True, Translation="0,0.2,0", Rotation="0,0,0", Scale="1,1,1") -> (Pose) {
    alphaInputType=Bool
    rotationMode=BMM_Additive
    rotationSpace=BCS_BoneSpace
    scaleMode=BMM_Additive
    scaleSpace=BCS_BoneSpace
    translationMode=BMM_Additive
    translationSpace=BCS_BoneSpace
    alphaBoolBlend {blendOption=Linear}
    boneToModify {boneName=dyn_drawString03_l}
    translation {y=0.2}
   }
   n11 = "Transform (Modify) Bone - Bone: dyn_drawString01_l_MSTR"!(ComponentPose=@n10, bAlphaBoolEnabled:bool=True, Translation="0,0.2,0", Rotation="0,5,0", Scale="1,1,1") -> (Pose) {
    alphaInputType=Bool
    rotationMode=BMM_Additive
    rotationSpace=BCS_BoneSpace
    scaleMode=BMM_Additive
    scaleSpace=BCS_BoneSpace
    translationMode=BMM_Additive
    translationSpace=BCS_BoneSpace
    alphaBoolBlend {blendOption=Linear, blendOutTime=5}
    boneToModify {boneName=dyn_drawString01_l_MSTR}
    rotation {yaw=5}
    translation {y=0.2}
   }
   n12 = Get bIsAccelerating2D()
   n13 = "Transform (Modify) Bone - Bone: dyn_drawString01_l_MSTR"(ComponentPose=@n11, bAlphaBoolEnabled=@n12, Translation="0,0.2,0", Rotation="-5,10,0", Scale="1,1,1") {
    alphaInputType=Bool
    rotationMode=BMM_Additive
    rotationSpace=BCS_BoneSpace
    scaleMode=BMM_Additive
    scaleSpace=BCS_BoneSpace
    translationMode=BMM_Additive
    translationSpace=BCS_BoneSpace
    alphaBoolBlend {blendOption=Linear, blendOutTime=10}
    boneToModify {boneName=dyn_drawString01_l_MSTR}
    rotation {pitch=-5, yaw=10}
    translation {y=0.2}
   }
   n14 = "Transform (Modify) Bone - Bone: dyn_drawString01_r_MASTR"!(ComponentPose=@n13, bAlphaBoolEnabled:bool=@n12, Translation="0,0.2,0", Rotation="0,10,0", Scale="1,1,1") -> (Pose) {
    alphaInputType=Bool
    rotationMode=BMM_Additive
    rotationSpace=BCS_BoneSpace
    scaleMode=BMM_Additive
    scaleSpace=BCS_BoneSpace
    translationMode=BMM_Additive
    translationSpace=BCS_BoneSpace
    alphaBoolBlend {blendOption=Linear, blendOutTime=10}
    boneToModify {boneName=dyn_drawString01_r_MASTR}
    rotation {yaw=10}
    translation {y=0.2}
   }
   Output Pose(Result=Component To Local(ComponentPose=@n14)) {layerGroup=None}
  ]}
 EventGraph {nodes [Event BlueprintUpdateAnimation()]}
]
variables []
